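# Niagara system for the Crisp Season drum kit's idle effect, spawning small, short-lived drip sprites from the instrument's skeletal mesh.
# A single emitter partitions particles three ways, using the index to pick a bone or socket and set the visibility tag across three sprite renderers.
# Drips inherit some motion, then fall under gravity and drag while fading and resizing.
# System-level values drive time-of-day tint, speed-based alpha and owner scale.
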NiagaraSystem /SparksCosmetics/Instrument/DrumKit/CrispSeason/Effects/NS_DrumKit_CrispSeason_Idle_FX
dependencies {
 usedDataInterfaces [
  &NiagaraDataInterfaceCurve
  &NiagaraDataInterfaceVector2DCurve
  &NiagaraDataInterfaceArrayColor
  &NiagaraDataInterfaceSkeletalMesh
  &NiagaraDataInterfaceColorCurve
 ]
 usedDynamicInputs [
  &ColorFromCurve
  &MakeFloatFromVector
  &Divide_Float
  &VectorLength
 ]
 usedModules [
  &self:SystemSpawnScript.SystemScriptSource.SystemScriptGraph.NiagaraNodeAssignment_3.SetVariables_#1
  &self:SystemSpawnScript.SystemScriptSource.SystemScriptGraph.NiagaraNodeAssignment_1.SetVariables_#2
  &SystemState
  &self:SystemSpawnScript.SystemScriptSource.SystemScriptGraph.NiagaraNodeAssignment_0.SetVariables_#3
  &self:SystemSpawnScript.SystemScriptSource.SystemScriptGraph.NiagaraNodeAssignment_6.SetVariables_#4
  &self:SystemSpawnScript.SystemScriptSource.SystemScriptGraph.NiagaraNodeAssignment_2.SetVariables_#5
  &self:SystemSpawnScript.SystemScriptSource.SystemScriptGraph.NiagaraNodeAssignment_4.SetVariables_#6
  &self:SystemSpawnScript.SystemScriptSource.SystemScriptGraph.NiagaraNodeAssignment_5.SetVariables_#7
 ]
}
emitters {DripSprites {
  emitterUpdateScript [
   &EmitterState()
   &SpawnRate(Spawn Probability=0.5, SpawnRate=3, Use Spawn Probability=true)
  ]
  particleSpawnScript [
   &PartitionParticles(Number Of Partitions=3, Partition Mode:ENiagara_PartitionMode_v2=::NewEnumerator5) {module=PartitionParticles001}
   /SetVariables(Particles.VisibilityTag=$Particles.PartitionParticles001.Partition) {isSetParameters=true}
   &InitializeParticle(
    Color=&MultiplyLinearColor(Linear Color=$System.TOD Color, Scale Factor=<1 1 1 0.9>)
    Lifetime Max=0.4
    Lifetime Min=0.3
    Lifetime Mode=::NewEnumerator1
    Position Offset=<0 0 -0.5>
    Position Offset Coordinate Space=::World
    Sprite Size Mode=::NewEnumerator1
    Uniform Sprite Size Max=2
    Uniform Sprite Size Min=1.5
    UsePositionOffset=true
    Write Material Random=true
   )
   &SkeletalMeshLocation(
    Bone / Socket Index=$Particles.PartitionParticles001.Partition
    MaskSpawnGroup=false
    Mesh Sampling Type=::NewEnumerator5
    Orientation Sampling=::NewEnumerator1
    Position Sampling=::NewEnumerator1
    Sampled Position Offset=&Multiply_VectorByFloat(Vector=<0 0 -0.5>, Float=$System.EngineOwnerScale)
    Sampled Position Offset Coordinate Space:ENiagaraSkelMeshTransforms=::NewEnumerator1
    Skeletal Mesh=$System.Skeletal Mesh
    Socket Sampling Mode=::NewEnumerator5
   ) {module=SkeletalMeshLocation001}
   /SetVariables(Particles.Color=&MultiplyLinearColor(Linear Color=$Particles.Color, Scale Factor=&MakeVector4(X=1, Y=1, Z=1, W=$System.SystemSpeedAlpha))) {isSetParameters=true}
   &InheritVelocity(
    Inherited Velocity Amount Scale=<0.15 0.15 0.15>
    Inherited Velocity Speed Limit=150
    Source Speed Threshold=15
    Speed Limit=true
    UseSpeedThreshold=true
   )
   &PartitionParticles()
  ]
  particleUpdateScript [
   &ParticleState()
   &CameraOffset(Camera Offset Amount=&FloatFromCurve(
     FloatCurve=DataInterface() {propertyValues {Curve {
        defaultValue=1
        keys [{interpMode=RCIM_Cubic, tangentMode=RCTM_User, time=0.2, value=1}, {interpMode=RCIM_Cubic, tangentMode=RCTM_User, time=1, value=0}]
       }}}
     Scale Curve=6
    ))
   &Drag(Drag=&RandomRangeFloat(Minimum=1, Maximum=7), Use Linear Drag=true, Use Rotational Drag=false)
   &GravityForce(Gravity=<0 0 -500>)
   &ScaleColor(
    Scale Alpha=&Multiply_Float(
     A=&FloatFromCurve(FloatCurve=DataInterface() {propertyValues {Curve {
         defaultValue=0
         keys [
          <at 0 0 RCIM_Cubic>
          {arriveTangentWeight=0.1258, interpMode=RCIM_Cubic, leaveTangentWeight=0.1887, tangentMode=RCTM_User, time=0.3, value=0.9621}
          {arriveTangentWeight=0.1, interpMode=RCIM_Cubic, leaveTangentWeight=0.1667, tangentMode=RCTM_User, time=0.5, value=1}
          {interpMode=RCIM_Cubic, tangentMode=RCTM_User, time=1, value=0}
         ]
        }}})
     B=$System.SystemSpeedAlpha
    )
    ScaleA=true
    ScaleRGB=true
   )
   &ScaleSpriteSize(
    Uniform Curve Scale=$System.EngineOwnerScale
    Uniform Curve Sprite Scale=DataInterface() {propertyValues {Curve {keys [
        {arriveTangent=2.753, interpMode=RCIM_Cubic, leaveTangent=2.753, tangentMode=RCTM_User, time=0, value=0.3494}
        {arriveTangent=0.143, interpMode=RCIM_Cubic, leaveTangent=0.143, tangentMode=RCTM_User, time=0.3012, value=0.9559}
        {interpMode=RCIM_Cubic, tangentMode=RCTM_User, time=1, value=0.1882}
       ]}}}
   ) {module=ScaleSpriteSize001}
   &SolveForcesAndVelocity(Clamp Velocity=false, Limit Acceleration=false)
  ]
  rendererClasses [&NiagaraSpriteRendererProperties]
  renderers [
   {rendererClass=&NiagaraSpriteRendererProperties}
   {rendererClass=&NiagaraSpriteRendererProperties, rendererIndex=1}
   {rendererClass=&NiagaraSpriteRendererProperties, rendererIndex=2}
  ]
 }}
summary {
 systemName=NS_DrumKit_CrispSeason_Idle_FX
 emitters [{emitterName=DripSprites, rendererClasses [&NiagaraSpriteRendererProperties]}]
 userVariables []
}
